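# Additive Niagara sprite material for the Caretaker snap effect, painting a tinted, decal-like chromatic shimmer over the scene behind it.
# Screen-aligned wave texture scrolls by a dynamic parameter, warped by noise, then tinted by color and particle color.
# Inverted fresnel from scene-texture world normals shapes color and opacity, alongside a depth-behind mask; Forward and Mobile paths skip both.
# A radial sphere mask and near-camera fade soften opacity.
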Material /Caretaker/FX/Snap/M_Caretaker_Snap_Chromatic_Particle_Decal
graph {
 properties {
  bEnableResponsiveAA=true
  bUsedWithNiagaraSprites=true
  blendMode=BLEND_Additive
  pixelDepthOffsetMode=PDOM_Legacy
  usedShadingModels=Default Lit
 }
 nodes [
  n1 = MaterialFunctionCall(
   Brightness=ScalarParameter() {defaultValue=1, parameterName=FresnelBrightness}
   Exponent=ScalarParameter() {defaultValue=3, parameterName=FresnelExponent}
   Camera Vector=ComponentMask(None=SceneTexture() {sceneTextureId=PPI_WorldNormal}.Color) {b=true, g=true, r=true}
  ) {materialFunction=&MF_Fresnel_FX}
  n2 = MaterialFunctionCall!(TextureResolution=ScalarParameter() {defaultValue=256, parameterName=NebulaTextureResolution}) {materialFunction=/Engine/Functions/Engine_MaterialFunctions02/Texturing/ScreenAlignedPixelToPixelUVs}
  n3 = VectorParameter() {parameterName=Flow UV, defaultValue=<rgba 1 1 0 0>}
  n4 = DynamicParameter() {paramNames [Upanner,Vpanner,Param3,Param4]}
  n5 = Add(A=Multiply(A=@n2.Result, B=AppendVector(A=@n3.R, B=@n3.R)), B=Multiply(A=Time(), B=AppendVector(A=@n4.Upanner, B=@n4.Upanner)))
  n6 = VectorParameter() {parameterName=Noise UV, defaultValue=<rgba 1 1 0 0>}
  n7 = AppendVector(A=@n6.R, B=@n6.R)
  n8 = Time()
  n9 = AppendVector(A=@n6.B, B=@n6.B)
  n10 = TextureSample(UVs=Add(A=Multiply(A=@n2.Result, B=@n7), B=Multiply(A=@n8, B=@n9))) {samplerType=SAMPLERTYPE_LinearColor, texture=&HighContrastNoise}
  n11 = Multiply(A=Add(A=ComponentMask(None=@n10.RGB) {g=true, r=true}) {constB=-0.2}) {constB=0.2}
  n12 = Desaturation(
   None=TextureSample(UVs=Add(A=@n5, B=@n11)) {samplerType=SAMPLERTYPE_LinearColor, texture=/Game/Effects/Fort_Effects/Textures/T_WavePattern_Chromatic}.RGB
   Fraction=ScalarParameter() {defaultValue=-1, parameterName=Desaturation}
  )
  n13 = Add(A=@n1.Fresnel Inverted, B=@n12)
  n14 = ShadingPathSwitch(Default=@n13, Deferred=@n13, Forward=@n12, Mobile=@n12)
  n15 = VectorParameter() {parameterName=Color, defaultValue=<rgba 0.7484 0.0452 1.5 1>}
  n16 = ParticleColor()
  MaterialFunctionCall(Emissive=Multiply(A=Multiply(A=@n14, B=@n15.RGB) {constA=1}, B=@n16.RGB)) {materialFunction=&MF_FX_EmissiveAdjustments}
  n17 = MaterialFunctionCall!(
   Distance Bias=ScalarParameter() {parameterName=Distance Bias}
   Distance Divisor=ScalarParameter() {defaultValue=50, parameterName=Distance Divisor}
   Scene Depth
   Pixel Depth
  ) {materialFunction=/Game/Effects/Fort_Effects/Materials/_Functions/MF_DistanceBehindTranslucency_01}
  n18 = ScalarParameter() {parameterName=DEBUG}
  n19 = Multiply(A=@n1.Fresnel Inverted, B=Multiply(A=@n16.A, B=LinearInterpolate(A=@n17.Normalized Inverted, Alpha=@n18)))
  n20 = TextureCoordinate()
  n21 = Constant() {r=0.5}
  n22 = ScalarParameter() {parameterName=Hardness}
  Multiply(
   A=Multiply(A=ShadingPathSwitch(Default=@n19, Deferred=@n19, Forward=@n16.A, Mobile=@n16.A), B=SphereMask(A=@n20, B=@n21, Radius=@n21, Hardness=@n22))
   B=MaterialFunctionCall(Distance=ScalarParameter() {parameterName=Near Camera Fade Distance}) {materialFunction=&Near_Camera_Fade}.Result
  )
 ]
}
