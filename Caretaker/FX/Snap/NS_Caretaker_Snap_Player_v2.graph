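# Niagara effect of glowing tethers from the Caretaker to eight points on a target skeletal mesh, apparently the player's.
# String ribbons extend between both meshes' bones, with pendulum-constraint slack that shrinks with distance.
# Lasso ribbons ring each target bone via torus sampling, with flare sprites and GPU sparks at attachments.
# System timelines drive extension, tightening and a final phase that retracts strings, widens lassos and fades strings and flares.
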NiagaraSystem /Caretaker/FX/Snap/NS_Caretaker_Snap_Player_v2
dependencies {
 usedDataInterfaces [
  &NiagaraDataInterfaceArrayFloat2
  &NiagaraDataInterfaceCurve
  &NiagaraDataInterfaceArrayFloat
  &NiagaraDataInterfaceColorCurve
  &NiagaraDataInterfaceVectorCurve
  &NiagaraDataInterfaceVector2DCurve
  &NiagaraDataInterfaceArrayFloat3
  &NiagaraDataInterfaceArrayColor
  &NiagaraDataInterfaceArrayInt32
 ]
 usedDynamicInputs []
 usedModules [
  &self:SystemSpawnScript.SystemScriptSource.SystemScriptGraph.NiagaraNodeAssignment_1.SetVariables_#1
  &SystemState
  &Timeline
  &SampleSkeletalMesh
  &self:SystemSpawnScript.SystemScriptSource.SystemScriptGraph.NiagaraNodeAssignment_0.SetVariables_#2
 ]
}
emitters {
 Connect_Flare {
  emitterUpdateScript [
   &EmitterState(
    Inactive Response=::NewEnumerator1
    Life Cycle Mode=::NewEnumerator1
    Loop Behavior=::NewEnumerator1
    Loop Delay=0.2
    Loop Duration Mode=::NewEnumerator1
    UseLoopDelay=true
   )
   &SpawnBurst_Instantaneous(Spawn Count=8, Use Spawn Probability=false, UseLoopCountLimit=false)
  ]
  particleSpawnScript [
   &PartitionParticles(Number Of Partitions=8)
   &InitializeParticle(Color=<rgba 10 0 0.8863 2>, Sprite Size Mode=::NewEnumerator3, Uniform Sprite Size=30, UsePositionOffset=false, Write Material Random=true)
  ]
  particleUpdateScript [
   &ParticleState(Kill Particles When Lifetime Has Elapsed=false, Let Infinitely Lived Particles Die When Emitter Deactivates=true)
   &SkeletalMeshLocation(
    Bone / Socket Index=$Particles.PartitionParticles.Partition
    Bone Sampling Mode=::NewEnumerator4
    MaskSpawnGroup=false
    Skeletal Mesh=$System.Skeletal Mesh
   )
   &TorusLocation(
    Align Torus With Axis=true
    Large Radius:NiagaraFloat=&Waveform(
     "Waveform [1]"=::NewEnumerator2
     "[1] Frequency"=1
     "[1] Amplitude Min/Max"=&SelectVector2DFromArray(
      Direct Array Index=$Particles.PartitionParticles.Partition
      Vector 2D Selection Array=DataInterface() {propertyValues {FloatData [<8 12>, <8 12>, <8 12>, <8 12>, <20 22>, <8 12>, <8 12>, <8 12>]}}
     )
     "[2] Phase"=$Emitter.Age
     "[3] Phase"=$Emitter.Age
    )
    MaskSpawnGroup=false
    Torus Axis=&RotateVectorByQuaternion(Vector To Rotate=<1 0 0>, Quaternion=$Particles.SkeletalMeshLocation.SampledRotation)
    Torus Distribution Mode:ENiagaraTorusDistributionMode=::NewEnumerator1
    Torus Mode=::NewEnumerator2
    U Position:NiagaraFloat=&SelectFloatFromArray(
     Direct Array Index=$Particles.PartitionParticles.Partition
     Float Selection Array=DataInterface() {propertyValues {FloatData [0,0.75,0,0.75,0.75,-0.25,0.395,0.75]}}
    )
   )
   &ScaleColor(
    Scale Alpha=&FloatFromCurve(
     FloatCurve=DataInterface() {propertyValues {Curve {keys [
         {interpMode=RCIM_Cubic, leaveTangentWeight=33.203, tangentMode=RCTM_User, time=0, value=1}
         {arriveTangent=-3.1117, interpMode=RCIM_Cubic, leaveTangent=-3.1117, tangentMode=RCTM_User, time=1, value=0}
        ]}}}
     CurveIndex=$System.Timeline002.NormalizedTime
    )
    Scale RGB=&VectorFromFloat(Value=&FloatFromCurve(
      FloatCurve=DataInterface() {propertyValues {Curve {keys [
          {arriveTangent=176.0875, interpMode=RCIM_Cubic, leaveTangent=176.0875, tangentMode=RCTM_User, time=0, value=1}
          <at 0.2 25 RCIM_Cubic>
          <at 1 0 RCIM_Cubic>
         ]}}}
      CurveIndex=$System.Timeline002.NormalizedTime
     ))
    ScaleA=true
    ScaleRGB=true
   )
   &DynamicMaterialParameters(
    Index 0 Param 1=0.125
    Index 0 Param 2=1
    Index 0 Param 3=1
    Index 0 Param 4=1
    Module.Param0WriteEnabled=true
    Module.Param1WriteEnabled=true
    Module.Param2WriteEnabled=true
    Module.Param3WriteEnabled=true
   )
   &ScaleSpriteSize(
    Uniform Curve Index=$System.Timeline002.NormalizedTime
    Uniform Curve Sprite Scale=DataInterface() {propertyValues {Curve {keys [
        {arriveTangent=4.0586, interpMode=RCIM_Cubic, leaveTangent=4.0586, leaveTangentWeight=3.128, tangentMode=RCTM_User, time=0, value=1}
        {arriveTangentWeight=3.3754, interpMode=RCIM_Cubic, tangentMode=RCTM_User, time=1, value=3}
       ]}}}
   )
  ]
  rendererClasses [&NiagaraSpriteRendererProperties]
  renderers [{rendererClass=&NiagaraSpriteRendererProperties}]
 }
 Lasso {
  emitterUpdateScript [
   &EmitterState(
    Inactive Response=::NewEnumerator1
    Life Cycle Mode=::NewEnumerator1
    Loop Behavior=::NewEnumerator1
    Loop Delay=0.21
    Loop Duration Mode=::NewEnumerator1
    UseLoopDelay=true
   )
   &SpawnBurst_Instantaneous(Spawn Count=80, Use Spawn Probability=false, UseLoopCountLimit=false)
  ]
  particleSpawnScript [
   &InitializeRibbon(Color=<rgba 5 0 0.4431 1>, Ribbon Width=12, UsePositionOffset=false, Write Material Random=true)
   &PartitionParticles(
    Normalized Index In Partition Includes One=true
    Number Of Partitions=8
    Particles Per Partition=10
    Partition Mode:ENiagara_PartitionMode=::NewEnumerator1
    Write Partition Index to Ribbon Link Order:ENiagara_AttributeSamplingApplyOutput=::NewEnumerator0
    Write Partition to Ribbon ID=::NewEnumerator0
   )
  ]
  particleUpdateScript [
   &ParticleState(Kill Particles When Lifetime Has Elapsed=false, Let Infinitely Lived Particles Die When Emitter Deactivates=true)
   /SetVariables(Particles.Random Offset:Vector3f=&MakeVector(
     X=&Waveform(
      "Waveform [1]"=::NewEnumerator3
      "[1] Frequency"=1
      "[1] Compound Frequency"=1
      "[1] Amplitude Min/Max"=<-5 5>
      "[2] Phase"=$Emitter.Age
      "[3] Phase"=$Emitter.Age
     )
     Y=0
     Z=0
    )) {isSetParameters=true}
   &SkeletalMeshLocation(
    Bone / Socket Index=$Particles.PartitionParticles.Partition
    Bone Sampling Mode=::NewEnumerator4
    MaskSpawnGroup=false
    Sampled Position Offset=&Lerp_Vector(
     A=&Multiply_VectorByFloat(Vector=$Particles.Random Offset, Float=-1)
     B=$Particles.Random Offset
     Alpha=$Particles.PartitionParticles.Normalized Index In Partition
    )
    Skeletal Mesh=$System.Skeletal Mesh
   )
   &TorusLocation(
    Align Torus With Axis=true
    Large Radius:NiagaraFloat=&Lerp_Float(
     A=&Waveform(
      "Waveform [1]"=::NewEnumerator2
      "[1] Frequency"=1
      "[1] Amplitude Min/Max"=&SelectVector2DFromArray(
       Direct Array Index=$Particles.PartitionParticles.Partition
       Vector 2D Selection Array=DataInterface() {propertyValues {FloatData [<8 12>, <8 12>, <8 12>, <8 12>, <20 22>, <8 12>, <8 12>, <8 12>]}}
      )
      "[2] Phase"=$Emitter.Age
      "[3] Phase"=$Emitter.Age
     )
     B=&MakeCustomFloatFromBool(
      Bool=&SetBoolByIntComparison(A=$Particles.PartitionParticles.Partition, B=4, Comparison Type=::NewEnumerator4)
      True Float=50
      False Float=24
     )
     Alpha=&FloatFromCurve(FloatCurve=DataInterface() {propertyValues {Curve {keys [<at 0.5 0>,<at 1 1>]}}}, CurveIndex=$System.Timeline002.NormalizedTime)
    )
    MaskSpawnGroup=false
    Torus Axis=&RotateVectorByQuaternion(Vector To Rotate=<1 0 0>, Quaternion=$Particles.SkeletalMeshLocation.SampledRotation)
    Torus Distribution Mode:ENiagaraTorusDistributionMode=::NewEnumerator1
    Torus Mode=::NewEnumerator2
    U Position:NiagaraFloat=&FloatFromCurve(
     FloatCurve=DataInterface() {propertyValues {Curve {keys [
         {arriveTangent=3.963, interpMode=RCIM_Cubic, leaveTangent=3.963, tangentMode=RCTM_User, time=0, value=0}
         {arriveTangent=1.3219, interpMode=RCIM_Cubic, leaveTangent=1.3219, tangentMode=RCTM_User, time=1, value=3}
        ]}}}
     CurveIndex=$Particles.PartitionParticles.Normalized Index In Partition
    )
   )
   /SetVariables(Particles.RibbonFacing=$Output.TorusLocation.TorusNormal) {isSetParameters=true}
   &DynamicMaterialParameters(
    Index 0 Param 2=&Lerp_Float(
     A=&Waveform(
      "Waveform [1]"=::NewEnumerator2
      "[1] Frequency"=1.333
      "[1] Amplitude Min/Max"=<-0.5 0.5>
      "[2] Phase"=$Emitter.Age
      "[3] Phase"=$Emitter.Age
     )
     B=-1
     Alpha=&FloatFromCurve(FloatCurve=DataInterface() {propertyValues {Curve {keys [<at 0.5 0>,<at 1 1>]}}}, CurveIndex=$System.Timeline002.NormalizedTime)
    )
    Index 0 Param 3=4
    Index 0 Param 4=&RandomRangeFloat(Minimum=0, Maximum=1)
    Module.Param0WriteEnabled=false
    Module.Param1WriteEnabled=true
    Module.Param2WriteEnabled=true
    Module.Param3WriteEnabled=true
   )
  ]
  rendererClasses [&NiagaraRibbonRendererProperties]
  renderers [{rendererClass=&NiagaraRibbonRendererProperties}]
 }
 Long_Sparks {
  simTarget=GPUComputeSim
  emitterUpdateScript [
   &EmitterState(Delay First Loop Only=true, Life Cycle Mode=::NewEnumerator1, Loop Delay=0.333, Loop Duration=1, UseLoopDelay=true)
   &SpawnRate(
    SpawnRate=&Multiply_Float(
     A=&RandomRangeFloat(Minimum=0, Maximum=200)
     B=&OneMinusFloat(Float=$System.Timeline002.NormalizedTime)
    )
    Use Spawn Probability=false
   )
  ]
  particleSpawnScript [
   &PartitionParticles(Number Of Partitions=8)
   &InitializeParticle(
    Color=<rgba 4.989 0.3013 10 1>
    Lifetime Max=0.35
    Lifetime Min=0.05
    Lifetime Mode=::NewEnumerator1
    Sprite Size Max=<2 2>
    Sprite Size Min=<1 1>
    Sprite Size Mode=::NewEnumerator2
    UsePositionOffset=false
    Write Material Random=true
   )
   &SkeletalMeshLocation(
    Bone / Socket Index=$Particles.PartitionParticles.Partition
    Bone Sampling Mode=::NewEnumerator4
    MaskSpawnGroup=false
    Position Sampling=::NewEnumerator1
    Skeletal Mesh=$System.Skeletal Mesh
   )
   &TorusLocation(
    Align Torus With Axis=true
    Large Radius:NiagaraFloat=&Waveform(
     "Waveform [1]"=::NewEnumerator2
     "[1] Frequency"=1
     "[1] Amplitude Min/Max"=&SelectVector2DFromArray(
      Direct Array Index=$Particles.PartitionParticles.Partition
      Vector 2D Selection Array=DataInterface() {propertyValues {FloatData [<8 12>, <8 12>, <8 12>, <8 12>, <20 22>, <8 12>, <8 12>, <8 12>]}}
     )
     "[2] Phase"=$Emitter.Age
     "[3] Phase"=$Emitter.Age
    )
    MaskSpawnGroup=false
    Torus Axis=&RotateVectorByQuaternion(Vector To Rotate=<1 0 0>, Quaternion=$Particles.SkeletalMeshLocation.SampledRotation)
    Torus Distribution Mode:ENiagaraTorusDistributionMode=::NewEnumerator1
    Torus Mode=::NewEnumerator2
    U Position:NiagaraFloat=&SelectFloatFromArray(
     Direct Array Index=$Particles.PartitionParticles.Partition
     Float Selection Array=DataInterface() {propertyValues {FloatData [0,0.75,0,0.75,0.75,-0.25,0.395,0.75]}}
    )
   )
   &AddVelocityInCone(
    Cone Axis=<0 -1 -1>
    Module.Use Velocity Falloff On Cone Axis=false
    Velocity Strength=&RandomRangeFloat(Minimum=75, Maximum=150)
   )
  ]
  particleUpdateScript [
   &ParticleState()
   &GravityForce()
   &VectorNoiseForce(Force Amount=1500, Module.Use Random Noise Time Interval=false)
   &Drag(Drag=&RandomRangeFloat(Minimum=1, Maximum=3), Use Linear Drag=true, Use Rotational Drag=false)
   &SolveForcesAndVelocity(Clamp Velocity=false, Limit Acceleration=false)
   &ScaleColor(
    Scale Alpha=&FloatFromCurve(FloatCurve=DataInterface() {propertyValues {Curve {keys [{interpMode=RCIM_Cubic, tangentMode=RCTM_User, time=0, value=1}, {interpMode=RCIM_Cubic, tangentMode=RCTM_User, time=1, value=0}]}}})
    Scale RGB=&VectorFromFloat(Value=&FloatFromCurve(
      FloatCurve=DataInterface() {propertyValues {Curve {keys [
          {arriveTangent=9.0725, interpMode=RCIM_Cubic, leaveTangent=9.0725, tangentMode=RCTM_User, time=0, value=0}
          {arriveTangent=-2.5249, interpMode=RCIM_Cubic, leaveTangent=-2.5249, tangentMode=RCTM_User, time=0.1408, value=0.9847}
          {arriveTangent=-0.0405, interpMode=RCIM_Cubic, leaveTangent=-0.0405, tangentMode=RCTM_User, time=1, value=0}
         ]}}}
      Scale Curve=100
     ))
    ScaleA=true
    ScaleRGB=true
   )
   &ScaleSpriteSizeBySpeed(Max Scale Factor=<1.5 3>, Min Scale Factor=<0.5 2>)
  ]
  rendererClasses [&NiagaraSpriteRendererProperties]
  renderers [{rendererClass=&NiagaraSpriteRendererProperties}]
 }
 String {
  emitterUpdateScript [
   &EmitterState(
    Inactive Response=::NewEnumerator1
    Life Cycle Mode=::NewEnumerator1
    Loop Behavior=::NewEnumerator1
    Loop Duration Mode=::NewEnumerator1
    UseLoopDelay=false
   )
   &SpawnBurst_Instantaneous(Spawn Count=80, Use Spawn Probability=false, UseLoopCountLimit=false)
  ]
  particleSpawnScript [
   &InitializeRibbon(Color=<rgba 3 3 3 3>, Ribbon Width=40, UsePositionOffset=false, Write Material Random=true)
   &PartitionParticles(
    Normalized Index In Partition Includes One=true
    Number Of Partitions=8
    Particles Per Partition=10
    Partition Mode:ENiagara_PartitionMode=::NewEnumerator1
    Write Partition Index to Ribbon Link Order:ENiagara_AttributeSamplingApplyOutput=::NewEnumerator0
    Write Partition to Ribbon ID=::NewEnumerator0
   )
  ]
  particleUpdateScript [
   &ParticleState(Kill Particles When Lifetime Has Elapsed=false, Let Infinitely Lived Particles Die When Emitter Deactivates=true)
   &SkeletalMeshLocation(
    Bone / Socket Index=$Particles.PartitionParticles.Partition
    Bone Sampling Mode=::NewEnumerator4
    MaskSpawnGroup=false
    Position Sampling=::NewEnumerator2
    Skeletal Mesh=$User.CaretakerMesh
   )
   &SkeletalMeshLocation(
    Bone / Socket Index=$Particles.PartitionParticles.Partition
    Bone Sampling Mode=::NewEnumerator4
    MaskSpawnGroup=false
    Position Sampling=::NewEnumerator2
    Skeletal Mesh=$System.Skeletal Mesh
   ) {module=SkeletalMeshLocation001}
   &self:TorusLocation(
    Align Torus With Axis=true
    Large Radius:NiagaraFloat=&Waveform(
     "Waveform [1]"=::NewEnumerator2
     "[1] Frequency"=1
     "[1] Amplitude Min/Max"=&SelectVector2DFromArray(
      Direct Array Index=$Particles.PartitionParticles.Partition
      Vector 2D Selection Array=DataInterface() {propertyValues {FloatData [<8 12>, <8 12>, <8 12>, <8 12>, <20 22>, <8 12>, <8 12>, <8 12>]}}
     )
     "[2] Phase"=$Emitter.Age
     "[3] Phase"=$Emitter.Age
    )
    MaskSpawnGroup=false
    Torus Axis:Vector3f=&RotateVectorByQuaternion(Vector To Rotate=<1 0 0>, Quaternion=$Particles.SkeletalMeshLocation001.SampledRotation)
    Torus Distribution Mode:ENiagaraTorusDistributionMode=::NewEnumerator1
    Torus Mode:ENiagaraTorusMode=::NewEnumerator2
    Torus Origin:NiagaraPosition=$Particles.SkeletalMeshLocation001.SampledPosition
    U Position:NiagaraFloat=&SelectFloatFromArray(
     Direct Array Index=$Particles.PartitionParticles.Partition
     Float Selection Array=DataInterface() {propertyValues {FloatData [0,0.75,0,0.75,0.75,-0.25,0.395,0.75]}}
    )
   ) {module=TorusLocation}
   /SetVariables(Particles.PositionOrigin:Vector3f=&Lerp_Vector(
     A=Linked() {linkedVariable {name=Particles.SkeletalMeshLocation.SampledPosition, type=NiagaraPosition}}
     B=&Lerp_Vector(
      A=&Add_Vector(
       A=&Lerp_Vector(
        A=Linked() {linkedVariable {name=Particles.SkeletalMeshLocation.SampledPosition, type=NiagaraPosition}}
        B=$Particles.TorusLocation.SampledPosition
        Alpha=$System.Timeline.NormalizedTime
       )
       B=&MakeCustomVectorFromBool(
        Bool=&SetBoolByFloatComparison(A=$System.StringDistance, B=0.25, Comparison Type=::NewEnumerator2)
        True Vector=&RandomVector(Vector Scale=150, Evaluation Type=::NewEnumerator1)
        False Vector=<0 0 0>
       )
      )
      B=Linked() {linkedVariable {name=Particles.SkeletalMeshLocation.SampledPosition, type=NiagaraPosition}}
      Alpha=$System.Timeline002.NormalizedTime
     )
     Alpha=$Particles.PartitionParticles.Normalized Index In Partition
    )) {isSetParameters=true}
   &GravityForce()
   &PendulumSetup(
    Module.Enable Spring Driver=false
    Pendulum Length:NiagaraFloat=&FloatFromCurve(
     FloatCurve=DataInterface() {propertyValues {Curve {keys [
         {arriveTangent=3.8991, interpMode=RCIM_Cubic, leaveTangent=3.8991, leaveTangentWeight=0.2364, tangentMode=RCTM_User, time=0, value=0}
         <at 0.5 1 RCIM_Cubic>
         {arriveTangent=-3.8444, arriveTangentWeight=0.4978, interpMode=RCIM_Cubic, leaveTangent=-3.8444, tangentMode=RCTM_User, time=1, value=0}
        ]}}}
     CurveIndex=$Particles.PartitionParticles.Normalized Index In Partition
     Scale Curve=&Lerp_Float(
      A=&Multiply_Float(
       A=125
       B=&RemapRange(
        Input Value To Remap=&VectorLength(Vector=&Subtract_Vector(
          A=$Particles.TorusLocation.SampledPosition
          B=Linked() {linkedVariable {name=Particles.SkeletalMeshLocation.SampledPosition, type=NiagaraPosition}}
         ))
        Input Min=0
        Input Max=$User.TightDistance
        Output Min=1
        Output Max=0
       )
      )
      B=0
      Alpha=$System.Timeline001.NormalizedTime
     )
    )
    Pendulum Pivot:Vector3f=$Particles.PositionOrigin
   )
   &SolveForcesAndVelocity(Clamp Velocity=false, Limit Acceleration=false)
   &PendulumConstraint(Calculate and Output Potential Energy=true)
   &CalculateAccurateVelocity()
   /SetVariables(Particles.RibbonLinkOrder:NiagaraFloat=&ReturnNormalizedExecIndex()) {isSetParameters=true}
   &DynamicMaterialParameters(
    Index 0 Param 3=&RandomRangeFloat(Minimum=1, Maximum=2)
    Index 0 Param 4=&RandomRangeFloat(Minimum=0, Maximum=1)
    Module.Param0WriteEnabled=false
    Module.Param1WriteEnabled=true
    Module.Param2WriteEnabled=true
    Module.Param3WriteEnabled=true
   )
   &ScaleRibbonWidth(Ribbon Width Scale=&FloatFromCurve(
     FloatCurve=DataInterface() {propertyValues {Curve {keys [
         {interpMode=RCIM_Cubic, leaveTangentWeight=0.3786, tangentMode=RCTM_User, time=0, value=1.25}
         {
          arriveTangent=-3.1323
          arriveTangentWeight=0.1929
          interpMode=RCIM_Cubic
          leaveTangent=-3.1323
          leaveTangentWeight=0.1929
          tangentMode=RCTM_User
          time=0.5
          value=0.5
         }
        ]}}}
     CurveIndex=$Particles.PartitionParticles.Normalized Index In Partition
    ))
   &ScaleColor(
    Scale Alpha=&FloatFromCurve(
     FloatCurve=DataInterface() {propertyValues {Curve {keys [<at 0.85 1 RCIM_Cubic>,<at 1 0 RCIM_Cubic>]}}}
     CurveIndex=$System.Timeline002.NormalizedTime
    )
    Scale RGB=&MakeVectorFromLinearColorRGB(Color=&ColorFromCurve(
      ColorCurve=DataInterface() {propertyValues {
        AlphaCurve {keys []}
        BlueCurve {keys [<at 0 1 RCIM_Cubic>,<at 1 0.013 RCIM_Cubic>]}
        GreenCurve {keys [<at 0 0.0301 RCIM_Cubic>,<at 1 0 RCIM_Cubic>]}
        RedCurve {keys [<at 0 0.4989 RCIM_Cubic>,<at 1 0.147 RCIM_Cubic>]}
       }}
      CurveIndex=$Particles.PartitionParticles.Normalized Index In Partition
     ))
    ScaleA=true
    ScaleRGB=true
   )
  ]
  rendererClasses [&NiagaraRibbonRendererProperties]
  renderers [{rendererClass=&NiagaraRibbonRendererProperties}]
 }
}
summary {
 systemName=NS_Caretaker_Snap_Player_v2
 emitters [
  {emitterName=Lasso, rendererClasses [&NiagaraRibbonRendererProperties]}
  {emitterName=Connect_Flare, rendererClasses [&NiagaraSpriteRendererProperties]}
  {emitterName=String, rendererClasses [&NiagaraRibbonRendererProperties]}
  {emitterName=Long_Sparks, simTarget=GPUComputeSim, rendererClasses [&NiagaraSpriteRendererProperties]}
 ]
 userVariables [
  User.BreakDistance {type=NiagaraFloat, defaultValue=2300, description=""}
  User.CaretakerMesh {
   type=NiagaraDataInterfaceSkeletalMesh
   defaultValue {
    struct=NiagaraExt_VariableValue_DataInterface
    value {dataInterface=&self:NiagaraDataInterfaceSkeletalMesh_0}
   }
  }
  User.SkeletalMesh {
   type=NiagaraDataInterfaceSkeletalMesh
   defaultValue {
    struct=NiagaraExt_VariableValue_DataInterface
    value {dataInterface=&self:NiagaraDataInterfaceSkeletalMesh_1}
   }
  }
  User.Tight {type=NiagaraBool}
  User.TightDistance {type=NiagaraFloat, defaultValue=1500, description=""}
 ]
}
